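# Physics asset of cape colliders for the F_MED_DirectContact female player cosmetic body.
# Capsules cover pelvis, spine, clavicles, head and feet, with paired capsules on spine_02 to spine_04; arm, thigh, calf and neck bodies have no shapes.
# Three capsule chains (center "dack", back L, back R) hang from pelvis via BACE roots.
# Constraints follow the bone hierarchy with all swing and twist motion free, so they impose no angular limits.
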PhysicsAsset /BRCosmetics/Characters/Player/Female/Medium/Bodies/F_MED_DirectContact/Meshes/F_MED_DirectContact_CapeColliders
bodies {
 calf_l []
 calf_r []
 clavicle_l [{shapeType=Capsule, length=6.5623, radius=7.5194, center=<3.9566 -1.0862 -2.1127>, rotation=<rot -77.5566 177.3915 7.7314>}]
 clavicle_r [{shapeType=Capsule, length=6.5623, radius=7.5194, center=<-3.9566 1.0862 2.1127>, rotation=<rot -77.5566 177.3915 -172.2686>}]
 dyn_dress_back_L_0 [{shapeType=Capsule, length=9.3158, radius=9.2672, rotation=<rot -0.0002 0.0042 39.0108>}]
 dyn_dress_back_L_1 [{shapeType=Capsule, length=9.3158, radius=9.2672, rotation=<rot 0 0.0043 39.0116>}]
 dyn_dress_back_L_2 [{shapeType=Capsule, length=9.3158, radius=9.2672, rotation=<rot 0.0001 0.0043 39.0118>}]
 dyn_dress_back_L_3 [{shapeType=Capsule, length=9.3158, radius=9.2672, rotation=<rot 0.0001 0.0043 39.0117>}]
 dyn_dress_back_L_4 [{shapeType=Capsule, length=9.3158, radius=9.2672, rotation=<rot 0.0001 0.0043 39.0116>}]
 dyn_dress_back_L_5 [{shapeType=Capsule, length=9.3158, radius=9.2672, rotation=<rot 0.0001 0.0043 39.0116>}]
 dyn_dress_back_L_BACE [{shapeType=Capsule, length=0.505, radius=0.505, shapeName=dyn_dress_back_L_BACE_capsule, rotation=<rot 0 -90 0>}]
 dyn_dress_back_R_0 [{shapeType=Capsule, length=9.3158, radius=9.2672, rotation=<rot 0.0002 0.0041 -32.2867>}]
 dyn_dress_back_R_1 [{shapeType=Capsule, length=9.3158, radius=9.2672, rotation=<rot 0 0.0043 -32.2875>}]
 dyn_dress_back_R_2 [{shapeType=Capsule, length=9.3158, radius=9.2672, rotation=<rot -0.0001 0.0043 -32.2877>}]
 dyn_dress_back_R_3 [{shapeType=Capsule, length=9.3158, radius=9.2672, rotation=<rot -0.0001 0.0043 -32.2876>}]
 dyn_dress_back_R_4 [{shapeType=Capsule, length=9.3158, radius=9.2672, rotation=<rot -0.0001 0.0043 -32.2875>}]
 dyn_dress_back_R_5 [{shapeType=Capsule, length=9.3158, radius=9.2672, rotation=<rot -0.0001 0.0043 -32.2875>}]
 dyn_dress_back_R_BACE [{shapeType=Capsule, length=0.505, radius=0.505, shapeName=dyn_dress_back_R_BACE_capsule, rotation=<rot 0 -90 0>}]
 dyn_dress_dack_0 [{
   shapeType=Capsule
   length=6.2733
   radius=4.5054
   shapeName=dyn_dress_dack_0_capsule
   center=<5.9804 2.5159 -2.766>
   rotation=<rot 0 43.0419 -74.9762>
  }]
 dyn_dress_dack_1 [{
   shapeType=Capsule
   length=24.8839
   radius=8.2245
   shapeName=dyn_dress_dack_1_capsule
   center=<6.6061 1.0057 -2.0847>
   rotation=<rot -0.996 -90 0.5733>
  }]
 dyn_dress_dack_2 [{
   shapeType=Capsule
   length=25.8263
   radius=12.6598
   shapeName=dyn_dress_dack_2_capsule
   center=<11.6024 0.8628 -1.5935>
   rotation=<rot 0 89.4936 4.3052>
  }]
 dyn_dress_dack_3 [{
   shapeType=Capsule
   length=25.4684
   radius=11.4081
   shapeName=dyn_dress_dack_3_capsule
   center=<7.7909 1.0841 -1.524>
   rotation=<rot 1.0531 -90 0.7726>
  }]
 dyn_dress_dack_4 [{
   shapeType=Capsule
   length=25.532
   radius=8.5962
   shapeName=dyn_dress_dack_4_capsule
   center=<8.5124 0.7475 -0.3205>
   rotation=<rot 0 -89.8884 -5.6039>
  }]
 dyn_dress_dack_5 [{
   shapeType=Capsule
   length=23.1667
   radius=3.1169
   shapeName=dyn_dress_dack_5_capsule
   center=<2.8021 0.9046 0.2858>
   rotation=<rot -0.9912 -90 -2.0016>
  }]
 dyn_dress_dack_BACE [{shapeType=Capsule, length=0.505, radius=0.505, shapeName=dyn_dress_dack_BACE_capsule, rotation=<rot 0 -90 0>}]
 foot_l [{shapeType=Capsule, length=11.0765, radius=8.0344, center=<-4.1811 6.0027 -0.338>, rotation=<rot 0.0388 1.4701 -89.3982>}]
 foot_r [{shapeType=Capsule, length=11.0765, radius=8.0344, center=<4.1811 -6.0027 0.338>, rotation=<rot 0.0388 1.4701 90.6018>}]
 head [{shapeType=Capsule, length=0, radius=18, center=<9.0697 0.0746 0>, rotation=<rot 0 -90 0>}]
 lowerarm_l []
 lowerarm_r []
 neck_02 []
 pelvis [{shapeType=Capsule, length=9.6042, radius=11.1441, center=<-2.4471 -2.5289 -0.1981>}]
 spine_01 [{shapeType=Capsule, length=12.2264, radius=16.3638, center=<-8.9021 -2.5544 0>}]
 spine_02 [
  {shapeType=Capsule, length=10.6157, radius=13.9336, center=<-3.5642 -3.7404 0>}
  {shapeType=Capsule, length=20.6875, radius=9.3125, center=<-1.8317 1.9362 0>, rotation=<rot 90 0 0>}
 ]
 spine_03 [
  {shapeType=Capsule, length=6.1863, radius=8.4854, center=<0.0574 0.4416 0>}
  {shapeType=Capsule, length=20.7188, radius=9.2812, center=<-1.2163 0.549 0>, rotation=<rot 90 180 -175.0732>}
 ]
 spine_04 [
  {shapeType=Capsule, length=9.0076, radius=10.3916, center=<5.6234 -0.6779 0>}
  {shapeType=Capsule, length=10.0312, radius=9.2812, center=<-1.2326 1.308 0>, rotation=<rot 90 0 0>}
 ]
 spine_05 [{shapeType=Capsule, length=9.678, radius=12.3571, center=<0.6969 1.91 0>}]
 thigh_l []
 thigh_r []
 upperarm_l []
 upperarm_r []
}
constraints [
 {
  bone1Name=spine_01
  bone2Name=pelvis
  swing1LimitDegrees=45
  swing1Motion=Free
  swing2LimitDegrees=45
  swing2Motion=Free
  twistLimitDegrees=45
  twistMotion=Free
 }
 {
  bone1Name=spine_02
  bone2Name=spine_01
  swing1LimitDegrees=45
  swing1Motion=Free
  swing2LimitDegrees=45
  swing2Motion=Free
  twistLimitDegrees=45
  twistMotion=Free
 }
 {
  bone1Name=spine_03
  bone2Name=spine_02
  swing1LimitDegrees=45
  swing1Motion=Free
  swing2LimitDegrees=45
  swing2Motion=Free
  twistLimitDegrees=45
  twistMotion=Free
 }
 {
  bone1Name=spine_04
  bone2Name=spine_03
  swing1LimitDegrees=45
  swing1Motion=Free
  swing2LimitDegrees=45
  swing2Motion=Free
  twistLimitDegrees=45
  twistMotion=Free
 }
 {
  bone1Name=spine_05
  bone2Name=spine_04
  swing1LimitDegrees=45
  swing1Motion=Free
  swing2LimitDegrees=45
  swing2Motion=Free
  twistLimitDegrees=45
  twistMotion=Free
 }
 {
  bone1Name=clavicle_l
  bone2Name=spine_05
  swing1LimitDegrees=45
  swing1Motion=Free
  swing2LimitDegrees=45
  swing2Motion=Free
  twistLimitDegrees=45
  twistMotion=Free
 }
 {
  bone1Name=upperarm_l
  bone2Name=clavicle_l
  swing1LimitDegrees=45
  swing1Motion=Free
  swing2LimitDegrees=45
  swing2Motion=Free
  twistLimitDegrees=45
  twistMotion=Free
 }
 {
  bone1Name=lowerarm_l
  bone2Name=upperarm_l
  swing1LimitDegrees=45
  swing1Motion=Free
  swing2LimitDegrees=45
  swing2Motion=Free
  twistLimitDegrees=45
  twistMotion=Free
 }
 {
  bone1Name=neck_02
  bone2Name=spine_05
  swing1LimitDegrees=45
  swing1Motion=Free
  swing2LimitDegrees=45
  swing2Motion=Free
  twistLimitDegrees=45
  twistMotion=Free
 }
 {
  bone1Name=head
  bone2Name=neck_02
  swing1LimitDegrees=45
  swing1Motion=Free
  swing2LimitDegrees=45
  swing2Motion=Free
  twistLimitDegrees=45
  twistMotion=Free
 }
 {
  bone1Name=calf_l
  bone2Name=thigh_l
  swing1LimitDegrees=45
  swing1Motion=Free
  swing2LimitDegrees=45
  swing2Motion=Free
  twistLimitDegrees=45
  twistMotion=Free
 }
 {
  bone1Name=thigh_l
  bone2Name=pelvis
  swing1LimitDegrees=45
  swing1Motion=Free
  swing2LimitDegrees=45
  swing2Motion=Free
  twistLimitDegrees=45
  twistMotion=Free
 }
 {
  bone1Name=foot_l
  bone2Name=calf_l
  swing1LimitDegrees=45
  swing1Motion=Free
  swing2LimitDegrees=45
  swing2Motion=Free
  twistLimitDegrees=45
  twistMotion=Free
 }
 {
  bone1Name=lowerarm_r
  bone2Name=upperarm_r
  swing1LimitDegrees=45
  swing1Motion=Free
  swing2LimitDegrees=45
  swing2Motion=Free
  twistLimitDegrees=45
  twistMotion=Free
 }
 {
  bone1Name=upperarm_r
  bone2Name=clavicle_r
  swing1LimitDegrees=45
  swing1Motion=Free
  swing2LimitDegrees=45
  swing2Motion=Free
  twistLimitDegrees=45
  twistMotion=Free
 }
 {
  bone1Name=clavicle_r
  bone2Name=spine_05
  swing1LimitDegrees=45
  swing1Motion=Free
  swing2LimitDegrees=45
  swing2Motion=Free
  twistLimitDegrees=45
  twistMotion=Free
 }
 {
  bone1Name=thigh_r
  bone2Name=pelvis
  swing1LimitDegrees=45
  swing1Motion=Free
  swing2LimitDegrees=45
  swing2Motion=Free
  twistLimitDegrees=45
  twistMotion=Free
 }
 {
  bone1Name=foot_r
  bone2Name=calf_r
  swing1LimitDegrees=45
  swing1Motion=Free
  swing2LimitDegrees=45
  swing2Motion=Free
  twistLimitDegrees=45
  twistMotion=Free
 }
 {
  bone1Name=calf_r
  bone2Name=thigh_r
  swing1LimitDegrees=45
  swing1Motion=Free
  swing2LimitDegrees=45
  swing2Motion=Free
  twistLimitDegrees=45
  twistMotion=Free
 }
 {
  bone1Name=dyn_dress_dack_BACE
  bone2Name=pelvis
  swing1LimitDegrees=45
  swing1Motion=Free
  swing2LimitDegrees=45
  swing2Motion=Free
  twistLimitDegrees=45
  twistMotion=Free
 }
 {
  bone1Name=dyn_dress_dack_0
  bone2Name=dyn_dress_dack_BACE
  swing1LimitDegrees=45
  swing1Motion=Free
  swing2LimitDegrees=45
  swing2Motion=Free
  twistLimitDegrees=45
  twistMotion=Free
 }
 {
  bone1Name=dyn_dress_dack_1
  bone2Name=dyn_dress_dack_0
  swing1LimitDegrees=45
  swing1Motion=Free
  swing2LimitDegrees=45
  swing2Motion=Free
  twistLimitDegrees=45
  twistMotion=Free
 }
 {
  bone1Name=dyn_dress_dack_2
  bone2Name=dyn_dress_dack_1
  swing1LimitDegrees=45
  swing1Motion=Free
  swing2LimitDegrees=45
  swing2Motion=Free
  twistLimitDegrees=45
  twistMotion=Free
 }
 {
  bone1Name=dyn_dress_dack_3
  bone2Name=dyn_dress_dack_2
  swing1LimitDegrees=45
  swing1Motion=Free
  swing2LimitDegrees=45
  swing2Motion=Free
  twistLimitDegrees=45
  twistMotion=Free
 }
 {
  bone1Name=dyn_dress_dack_4
  bone2Name=dyn_dress_dack_3
  swing1LimitDegrees=45
  swing1Motion=Free
  swing2LimitDegrees=45
  swing2Motion=Free
  twistLimitDegrees=45
  twistMotion=Free
 }
 {
  bone1Name=dyn_dress_dack_5
  bone2Name=dyn_dress_dack_4
  swing1LimitDegrees=45
  swing1Motion=Free
  swing2LimitDegrees=45
  swing2Motion=Free
  twistLimitDegrees=45
  twistMotion=Free
 }
 {
  bone1Name=dyn_dress_back_L_BACE
  bone2Name=pelvis
  swing1LimitDegrees=45
  swing1Motion=Free
  swing2LimitDegrees=45
  swing2Motion=Free
  twistLimitDegrees=45
  twistMotion=Free
 }
 {
  bone1Name=dyn_dress_back_L_0
  bone2Name=dyn_dress_back_L_BACE
  swing1LimitDegrees=45
  swing1Motion=Free
  swing2LimitDegrees=45
  swing2Motion=Free
  twistLimitDegrees=45
  twistMotion=Free
 }
 {
  bone1Name=dyn_dress_back_L_1
  bone2Name=dyn_dress_back_L_0
  swing1LimitDegrees=45
  swing1Motion=Free
  swing2LimitDegrees=45
  swing2Motion=Free
  twistLimitDegrees=45
  twistMotion=Free
 }
 {
  bone1Name=dyn_dress_back_L_2
  bone2Name=dyn_dress_back_L_1
  swing1LimitDegrees=45
  swing1Motion=Free
  swing2LimitDegrees=45
  swing2Motion=Free
  twistLimitDegrees=45
  twistMotion=Free
 }
 {
  bone1Name=dyn_dress_back_L_3
  bone2Name=dyn_dress_back_L_2
  swing1LimitDegrees=45
  swing1Motion=Free
  swing2LimitDegrees=45
  swing2Motion=Free
  twistLimitDegrees=45
  twistMotion=Free
 }
 {
  bone1Name=dyn_dress_back_L_4
  bone2Name=dyn_dress_back_L_3
  swing1LimitDegrees=45
  swing1Motion=Free
  swing2LimitDegrees=45
  swing2Motion=Free
  twistLimitDegrees=45
  twistMotion=Free
 }
 {
  bone1Name=dyn_dress_back_L_5
  bone2Name=dyn_dress_back_L_4
  swing1LimitDegrees=45
  swing1Motion=Free
  swing2LimitDegrees=45
  swing2Motion=Free
  twistLimitDegrees=45
  twistMotion=Free
 }
 {
  bone1Name=dyn_dress_back_R_BACE
  bone2Name=pelvis
  swing1LimitDegrees=45
  swing1Motion=Free
  swing2LimitDegrees=45
  swing2Motion=Free
  twistLimitDegrees=45
  twistMotion=Free
 }
 {
  bone1Name=dyn_dress_back_R_0
  bone2Name=dyn_dress_back_R_BACE
  swing1LimitDegrees=45
  swing1Motion=Free
  swing2LimitDegrees=45
  swing2Motion=Free
  twistLimitDegrees=45
  twistMotion=Free
 }
 {
  bone1Name=dyn_dress_back_R_1
  bone2Name=dyn_dress_back_R_0
  swing1LimitDegrees=45
  swing1Motion=Free
  swing2LimitDegrees=45
  swing2Motion=Free
  twistLimitDegrees=45
  twistMotion=Free
 }
 {
  bone1Name=dyn_dress_back_R_2
  bone2Name=dyn_dress_back_R_1
  swing1LimitDegrees=45
  swing1Motion=Free
  swing2LimitDegrees=45
  swing2Motion=Free
  twistLimitDegrees=45
  twistMotion=Free
 }
 {
  bone1Name=dyn_dress_back_R_3
  bone2Name=dyn_dress_back_R_2
  swing1LimitDegrees=45
  swing1Motion=Free
  swing2LimitDegrees=45
  swing2Motion=Free
  twistLimitDegrees=45
  twistMotion=Free
 }
 {
  bone1Name=dyn_dress_back_R_4
  bone2Name=dyn_dress_back_R_3
  swing1LimitDegrees=45
  swing1Motion=Free
  swing2LimitDegrees=45
  swing2Motion=Free
  twistLimitDegrees=45
  twistMotion=Free
 }
 {
  bone1Name=dyn_dress_back_R_5
  bone2Name=dyn_dress_back_R_4
  swing1LimitDegrees=45
  swing1Motion=Free
  swing2LimitDegrees=45
  swing2Motion=Free
  twistLimitDegrees=45
  twistMotion=Free
 }
]
