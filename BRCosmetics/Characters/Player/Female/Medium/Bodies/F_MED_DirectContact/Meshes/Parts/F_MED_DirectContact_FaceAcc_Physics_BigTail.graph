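# Physics asset for the big pigtail on the DirectContact female medium character's face accessory mesh, defining its hair chain and body collision shapes.
# Six dyn_Pigtail capsules chain from a small base linked to neck_02, with tight twist and swing limits that widen toward the tip.
# Pelvis, spine, neck, head and upper-arm shapes supply the body volumes, with arms angularly locked to spine_05.
# A separate small dyn_sheafl_BACE body also links to neck_02.
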PhysicsAsset /BRCosmetics/Characters/Player/Female/Medium/Bodies/F_MED_DirectContact/Meshes/Parts/F_MED_DirectContact_FaceAcc_Physics_BigTail
bodies {
 dyn_Pigtail_1 [{shapeType=Capsule, length=5.1697, radius=1.2416, shapeName=dyn_Pigtail_1_capsule, center=<3.0468 -0.7763 0>, rotation=<rot 90 0 0>}]
 dyn_Pigtail_2 [{shapeType=Capsule, length=4.4177, radius=1.9387, shapeName=dyn_Pigtail_2_capsule, center=<3.3633 -0.3257 0>, rotation=<rot 90 0 0>}]
 dyn_Pigtail_3 [{shapeType=Capsule, length=4.0375, radius=1.9566, shapeName=dyn_Pigtail_3_capsule, center=<3.8125 0.0845 0>, rotation=<rot 90 0 0>}]
 dyn_Pigtail_4 [{shapeType=Capsule, length=4.0901, radius=1.49, shapeName=dyn_Pigtail_4_capsule, center=<3.2512 0 0>, rotation=<rot 90 0 0>}]
 dyn_Pigtail_5 [{shapeType=Capsule, length=3.9339, radius=1.5324, shapeName=dyn_Pigtail_5_capsule, center=<3.3538 0 0>, rotation=<rot 90 0 0>}]
 dyn_Pigtail_6 [{shapeType=Capsule, length=3.047, radius=2.3548, shapeName=dyn_Pigtail_6_capsule, center=<2 0 0>, rotation=<rot 90 0 0>}]
 dyn_Pigtail_BACE [{shapeType=Capsule, length=0.505, radius=0.505, shapeName=dyn_Pigtail_BACE_capsule, rotation=<rot 0 -90 0>}]
 dyn_sheafl_BACE [{shapeType=Capsule, length=0.505, radius=0.505, shapeName=dyn_sheafl_BACE_capsule, rotation=<rot 0 -90 0>}]
 head [{shapeType=Sphere, radius=11, shapeName=head_sphere, center=<6.7981 0 0>}]
 neck_01 [{shapeType=Capsule, length=19.4596, radius=5.8732, shapeName=neck_01_capsule, center=<-2.2056 1.0927 0>}]
 neck_02 [{shapeType=Capsule, length=18.8547, radius=4.4204, shapeName=neck_02_capsule, center=<-1.8329 -0.16 0>}]
 pelvis [{shapeType=Capsule, length=16, radius=14, shapeName=pelvis_capsule}]
 spine_03 [{shapeType=Capsule, length=25.505, radius=10.5, shapeName=spine_03_capsule, center=<0 1.1444 0>, rotation=<rot 0 -90 0>}]
 spine_04 [
  {shapeType=Capsule, length=25.505, radius=10.5, shapeName=spine_04_capsule, center=<0 0.7917 0>, rotation=<rot 0 -90 0>}
  {shapeType=Capsule, length=29.5, radius=18, shapeName=spine_04_capsule, center=<-1.2785 11.407 0>, rotation=<rot -90 180 180>}
 ]
 spine_05 [{shapeType=Capsule, length=16.755, radius=9.225, shapeName=spine_05_capsule, center=<2.5115 0.8995 0>}]
 upperarm_l [{
   shapeType=Capsule
   length=23.1915
   radius=7.4419
   shapeName=upperarm_l_capsule
   center=<14.1088 0 1.1512>
   rotation=<rot -90 179.578 -179.5781>
  }]
 upperarm_r [{
   shapeType=Capsule
   length=23.0999
   radius=6.9001
   shapeName=upperarm_r_capsule
   center=<-17.4802 0.4156 0>
   rotation=<rot -90 179.9988 -179.9988>
  }]
}
constraints [
 {
  bone1Name=upperarm_l
  bone2Name=spine_05
  swing1LimitDegrees=45
  swing1Motion=Locked
  swing2LimitDegrees=45
  swing2Motion=Locked
  twistLimitDegrees=45
  twistMotion=Locked
 }
 {
  bone1Name=upperarm_r
  bone2Name=spine_05
  swing1LimitDegrees=45
  swing1Motion=Locked
  swing2LimitDegrees=45
  swing2Motion=Locked
  twistLimitDegrees=45
  twistMotion=Locked
 }
 {
  bone1Name=spine_05
  bone2Name=spine_04
  swing1LimitDegrees=45
  swing1Motion=Free
  swing2LimitDegrees=45
  swing2Motion=Free
  twistLimitDegrees=45
  twistMotion=Free
 }
 {
  bone1Name=dyn_Pigtail_BACE
  bone2Name=neck_02
  swing1LimitDegrees=45
  swing1Motion=Free
  swing2LimitDegrees=45
  swing2Motion=Free
  twistLimitDegrees=45
  twistMotion=Free
 }
 {bone1Name=dyn_Pigtail_1, bone2Name=dyn_Pigtail_BACE, swing1LimitDegrees=15, swing2LimitDegrees=15, twistLimitDegrees=1}
 {bone1Name=dyn_Pigtail_2, bone2Name=dyn_Pigtail_1, swing1LimitDegrees=15, swing2LimitDegrees=15, twistLimitDegrees=1}
 {bone1Name=dyn_Pigtail_3, bone2Name=dyn_Pigtail_2, swing1LimitDegrees=20, swing2LimitDegrees=20, twistLimitDegrees=1}
 {bone1Name=dyn_Pigtail_4, bone2Name=dyn_Pigtail_3, swing1LimitDegrees=30, swing2LimitDegrees=25, twistLimitDegrees=1}
 {bone1Name=dyn_Pigtail_5, bone2Name=dyn_Pigtail_4, swing1LimitDegrees=40, swing2LimitDegrees=25, twistLimitDegrees=1}
 {bone1Name=dyn_Pigtail_6, bone2Name=dyn_Pigtail_5, swing1LimitDegrees=60, swing2LimitDegrees=60, twistLimitDegrees=1}
 {
  bone1Name=dyn_sheafl_BACE
  bone2Name=neck_02
  swing1LimitDegrees=45
  swing1Motion=Free
  swing2LimitDegrees=45
  swing2Motion=Free
  twistLimitDegrees=45
  twistMotion=Free
 }
 {
  bone1Name=neck_01
  bone2Name=spine_05
  swing1LimitDegrees=45
  swing1Motion=Free
  swing2LimitDegrees=45
  swing2Motion=Free
  twistLimitDegrees=45
  twistMotion=Free
 }
 {
  bone1Name=neck_02
  bone2Name=neck_01
  swing1LimitDegrees=45
  swing1Motion=Free
  swing2LimitDegrees=45
  swing2Motion=Free
  twistLimitDegrees=45
  twistMotion=Free
 }
 {
  bone1Name=spine_03
  bone2Name=pelvis
  swing1LimitDegrees=45
  swing1Motion=Free
  swing2LimitDegrees=45
  swing2Motion=Free
  twistLimitDegrees=45
  twistMotion=Free
 }
 {
  bone1Name=spine_04
  bone2Name=spine_03
  swing1LimitDegrees=45
  swing1Motion=Free
  swing2LimitDegrees=45
  swing2Motion=Free
  twistLimitDegrees=45
  twistMotion=Free
 }
]
