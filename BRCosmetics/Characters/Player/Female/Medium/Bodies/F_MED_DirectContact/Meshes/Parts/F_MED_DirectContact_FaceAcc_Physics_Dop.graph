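# Physics asset for the face accessory of the F_MED_DirectContact female player cosmetic, covering a left and a right dynamic tail on the head.
# A head sphere anchors two mirrored chains, each made of a tiny base capsule and an offset outer capsule.
# Each base has all angular motion locked to the head.
# Each outer segment's joint to its base has a wide swing limit on one axis and narrow limits on the other swing and twist.
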PhysicsAsset /BRCosmetics/Characters/Player/Female/Medium/Bodies/F_MED_DirectContact/Meshes/Parts/F_MED_DirectContact_FaceAcc_Physics_Dop
bodies {
 dyn_tail_L_1 [{shapeType=Capsule, length=2, radius=1, shapeName=dyn_tail_L_1_capsule, center=<8.0043 0 0>, rotation=<rot 90 0 0>}]
 dyn_tail_L_BACE [{shapeType=Capsule, length=0.505, radius=0.505, shapeName=dyn_tail_L_BACE_capsule, rotation=<rot 0 -90 0>}]
 dyn_tail_R_1 [{shapeType=Capsule, length=2, radius=1, shapeName=dyn_tail_R_1_capsule, center=<8.0043 0 0>, rotation=<rot 90 0 0>}]
 dyn_tail_R_BACE [{shapeType=Capsule, length=0.505, radius=0.505, shapeName=dyn_tail_R_BACE_capsule, rotation=<rot 0 -90 0>}]
 head [{shapeType=Sphere, radius=6.9863, shapeName=head_sphere, center=<1.1275 2.2339 0>}]
}
constraints [
 {
  bone1Name=dyn_tail_L_BACE
  bone2Name=head
  swing1LimitDegrees=45
  swing1Motion=Locked
  swing2LimitDegrees=45
  swing2Motion=Locked
  twistLimitDegrees=45
  twistMotion=Locked
 }
 {
  bone1Name=dyn_tail_R_BACE
  bone2Name=head
  swing1LimitDegrees=45
  swing1Motion=Locked
  swing2LimitDegrees=45
  swing2Motion=Locked
  twistLimitDegrees=45
  twistMotion=Locked
 }
 {bone1Name=dyn_tail_L_1, bone2Name=dyn_tail_L_BACE, swing1LimitDegrees=5, swing2LimitDegrees=40, twistLimitDegrees=1}
 {bone1Name=dyn_tail_R_1, bone2Name=dyn_tail_R_BACE, swing1LimitDegrees=5, swing2LimitDegrees=40, twistLimitDegrees=1}
]
