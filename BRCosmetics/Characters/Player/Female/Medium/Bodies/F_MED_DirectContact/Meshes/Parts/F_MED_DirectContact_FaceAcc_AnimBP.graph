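# Post-process-style hair AnimBP for the DirectContact female FaceAcc part, animating its pigtail and braid bones.
# It copies the parent pose, layers in body-mesh shoulders, then selects cached cape, backpack or normal dynamics.
# Normal uses corrective offsets and RigidBody simulation, or constraints without physics; cape and backpack pin the pigtail to head/spine_05 before simulation.
# The EventGraph derives corrective alphas from head rotation and movement state, plus gravity override.
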AnimBlueprint /BRCosmetics/Characters/Player/Female/Medium/Bodies/F_MED_DirectContact/Meshes/Parts/F_MED_DirectContact_FaceAcc_AnimBP
target_skeleton=&/F_MED_DirectContact_FaceAcc_Skeleton
graphs [
 AnimGraph {nodes [
   n1 = Use cached pose 'Backpack Dynamics'()
   n2 = Use cached pose 'Normal Dynamics'()
   n3 = Get is Wearing Backbling() -> (is Wearing Backbling:bool)
   Output Pose(Result=Blend Poses by bool(
     BlendPose_0=Use cached pose 'Cape Dynamics'()
     BlendPose_1=Blend Poses by bool(BlendPose_0=@n1, BlendPose_1=@n2, bActiveValue=@n3) {blendType=Linear}
     bActiveValue=Get is Wearing Cape() -> (is Wearing Cape:bool)
    ) {blendType=Linear}) {layerGroup=None}
   n4 = Layered blend per bone(
    BasePose=Copy Pose From Mesh() {bCopyCurves=true, bUseAttachedParent=true}
    BlendPoses_0=Copy Pose From Mesh(SourceMeshComponent=Get Body_Mesh())
   ) {layerSetup [{branchFilters [{boneName=clavicle_l}, {boneName=clavicle_r}, {boneName=upperarm_r}, {boneName=upperarm_l}]}]}
   Save cached pose 'InputPose'(Pose=@n4) {cacheName=InputPose}
   n5 = Get Head Fwd Bwd Rot Crouching() -> (Head Fwd Bwd Rot Crouching:double)
   n6 = "Transform (Modify) Bone - Bone: dyn_Pigtail_BACE"(
    ComponentPose=Local To Component(LocalPose=Use cached pose 'InputPose'())
    Alpha=Get Hair Up() -> (Hair Up:double)
    Translation="0,0,0"
    Rotation="0,30,0"
    Scale="1,1,1"
   ) {
    rotationMode=BMM_Additive
    rotationSpace=BCS_BoneSpace
    alphaBoolBlend {blendOption=Linear, blendOutTime=0.5}
    boneToModify {boneName=dyn_Pigtail_BACE}
    rotation {yaw=30}
   }
   n7 = "Transform (Modify) Bone - Bone: dyn_Pigtail_BACE"(ComponentPose=@n6, Alpha=Get is Skydiving() -> (is Skydiving:double), Translation="-4,-4,0", Rotation="0,40,0", Scale="1,1,1") {
    rotationMode=BMM_Additive
    rotationSpace=BCS_BoneSpace
    translationMode=BMM_Additive
    translationSpace=BCS_BoneSpace
    alphaBoolBlend {blendOption=Linear}
    boneToModify {boneName=dyn_Pigtail_BACE}
    rotation {yaw=40}
    translation=<-4 -4>
   }
   n8 = "Transform (Modify) Bone - Bone: dyn_Pigtail_BACE"(ComponentPose=@n7, Alpha=Get is Parachuting Alpha() -> (is Parachuting Alpha:double), Translation="-4,-4,0", Rotation="0,0,0", Scale="1,1,1") {
    rotationMode=BMM_Additive
    rotationSpace=BCS_BoneSpace
    translationMode=BMM_Additive
    translationSpace=BCS_BoneSpace
    alphaBoolBlend {blendOption=Linear}
    alphaScaleBiasClamp {bInterpResult=true}
    boneToModify {boneName=dyn_Pigtail_BACE}
    translation=<-4 -4>
   }
   n9 = "Transform (Modify) Bone - Bone: dyn_Pigtail_BACE"(ComponentPose=@n8, Alpha=Get is Crouching() -> (is Crouching:double), Translation="-5.669,-2.071,0.604", Rotation="0,0,0", Scale="1,1,1") {
    rotationMode=BMM_Additive
    rotationSpace=BCS_BoneSpace
    translationMode=BMM_Additive
    translationSpace=BCS_BoneSpace
    alphaBoolBlend {blendOption=Linear}
    alphaScaleBiasClamp {bInterpResult=true}
    boneToModify {boneName=dyn_Pigtail_BACE}
    translation=<-5.669 -2.071 0.604>
   }
   n10 = "Transform (Modify) Bone - Bone: dyn_Pigtail_BACE"(
    ComponentPose=@n9
    Alpha=Get CrounchingFixHeadUp() -> (CrounchingFixHeadUp:double)
    Translation="-5.669,-2.071,0.604"
    Rotation="0,0,0"
    Scale="1,1,1"
   ) {
    rotationMode=BMM_Additive
    rotationSpace=BCS_BoneSpace
    translationMode=BMM_Additive
    translationSpace=BCS_BoneSpace
    alphaBoolBlend {blendOption=Linear}
    alphaScaleBiasClamp {bInterpResult=true}
    boneToModify {boneName=dyn_Pigtail_BACE}
    translation=<-5.669 -2.071 0.604>
   }
   n11 = "Transform (Modify) Bone - Bone: dyn_sheafl_BACE"(ComponentPose=@n10, Alpha=Get CrounchingFixHeadUp(), Translation="0.434,-0.768,0.024", Rotation="-0.078,-67.496,-0.015", Scale="1,1,1") {
    rotationMode=BMM_Additive
    rotationSpace=BCS_BoneSpace
    translationMode=BMM_Additive
    translationSpace=BCS_BoneSpace
    alphaBoolBlend {blendOption=Linear}
    alphaScaleBiasClamp {bInterpResult=true}
    boneToModify {boneName=dyn_sheafl_BACE}
    rotation=<rot -0.078 -67.496 -0.015>
    translation=<0.434 -0.768 0.024>
   }
   n12 = "Transform (Modify) Bone - Bone: dyn_Pigtail_BACE"(
    ComponentPose=@n11
    Alpha=Get ADS Up Crounching() -> (ADS Up Crounching:double)
    Translation="0,0,0"
    Rotation="-0.014,-29.152,0.013"
    Scale="1,1,1"
   ) {
    rotationMode=BMM_Additive
    rotationSpace=BCS_BoneSpace
    translationMode=BMM_Additive
    translationSpace=BCS_BoneSpace
    alphaBoolBlend {blendOption=Linear}
    boneToModify {boneName=dyn_Pigtail_BACE}
    rotation=<rot -0.014 -29.152 0.013>
   }
   n13 = Get ADS Up Crouch Moving() -> (ADS Up Crouch Moving:double)
   n14 = "Transform (Modify) Bone - Bone: dyn_Pigtail_BACE"(ComponentPose=@n12, Alpha=@n13, Translation="-5.563,0.887,0.754", Rotation="0.009,-20.246,-0.010", Scale="1,1,1") {
    rotationMode=BMM_Additive
    rotationSpace=BCS_BoneSpace
    translationMode=BMM_Additive
    translationSpace=BCS_BoneSpace
    alphaBoolBlend {blendOption=Linear}
    alphaScaleBiasClamp {bInterpResult=true}
    boneToModify {boneName=dyn_Pigtail_BACE}
    rotation=<rot 0.009 -20.246 -0.01>
    translation=<-5.563 0.887 0.754>
   }
   n15 = "Transform (Modify) Bone - Bone: dyn_sheafl_BACE"(ComponentPose=@n14, Alpha=@n13, Translation="1.445,-2.319,-0.539", Rotation="-0.579,-35.665,-0.685", Scale="1,1,1") {
    rotationMode=BMM_Additive
    rotationSpace=BCS_BoneSpace
    translationMode=BMM_Additive
    translationSpace=BCS_BoneSpace
    alphaBoolBlend {blendOption=Linear}
    alphaScaleBiasClamp {bInterpResult=true}
    boneToModify {boneName=dyn_sheafl_BACE}
    rotation=<rot -0.579 -35.665 -0.685>
    translation=<1.445 -2.319 -0.539>
   }
   n16 = "Transform (Modify) Bone - Bone: dyn_tail_R_BACE"(
    ComponentPose=@n15
    Alpha=Get ADS Up Side Braid Fix() -> (ADS Up Side Braid Fix:double)
    Translation="-1.673,-0.001,0.211"
    Rotation="-20.542,-0.020,0.003"
    Scale="1,1,1"
   ) {
    rotationMode=BMM_Additive
    rotationSpace=BCS_BoneSpace
    scaleMode=BMM_Additive
    scaleSpace=BCS_BoneSpace
    translationMode=BMM_Additive
    translationSpace=BCS_BoneSpace
    alphaBoolBlend {blendOption=Linear}
    boneToModify {boneName=dyn_tail_R_BACE}
    rotation=<rot -20.542 -0.0201 0.003>
    translation=<-1.673 -0.001 0.211>
   }
   n17 = "Transform (Modify) Bone - Bone: dyn_Pigtail_BACE"(ComponentPose=@n16, Alpha=@n5, Translation="-2.789,-2.107,0.157", Rotation="0.010,-19.863,0.020", Scale="1,1,1") {
    rotationMode=BMM_Additive
    rotationSpace=BCS_BoneSpace
    translationMode=BMM_Additive
    translationSpace=BCS_BoneSpace
    alphaBoolBlend {blendOption=Linear}
    alphaScaleBiasClamp {bMapRange=true, inRange {max=0, min=-25}, outRange {max=0, min=1}}
    boneToModify {boneName=dyn_Pigtail_BACE}
    rotation=<rot 0.01 -19.863 0.02>
    translation=<-2.789 -2.107 0.157>
   }
   n18 = "Transform (Modify) Bone - Bone: dyn_Pigtail_2"(ComponentPose=@n17, Alpha=@n5, Translation="-0.319,0.593,2.548", Rotation="0.073,28.065,-0.084", Scale="1,1,1") {
    rotationMode=BMM_Additive
    rotationSpace=BCS_BoneSpace
    translationMode=BMM_Additive
    translationSpace=BCS_BoneSpace
    alphaBoolBlend {blendOption=Linear}
    alphaScaleBiasClamp {bMapRange=true, inRange {max=0, min=-25}, outRange {max=0, min=1}}
    boneToModify {boneName=dyn_Pigtail_2}
    rotation=<rot 0.073 28.065 -0.084>
    translation=<-0.319 0.593 2.548>
   }
   n19 = RigidBody!(ComponentPose=@n18, Alpha, OverrideWorldGravity:Vector=Get GravityOverride(), ExternalForce) -> (Pose) {
    bOverrideWorldGravity=true
    overridePhysicsAsset=&/F_MED_DirectContact_FaceAcc_Physics_BigTail
    simulationSpace=BaseBoneSpace
    alphaBoolBlend {blendOption=Linear}
    baseBoneRef {boneName=neck_02}
    componentAppliedLinearAccClamp=<1000 650 400>
    componentLinearAccScale=<0.4 0.6 0.4>
    componentLinearVelScale=<0.8 0.4 0.2>
    overrideWorldGravity {z=-980}
   }
   n20 = RigidBody(ComponentPose=@n19, OverrideWorldGravity=Get GravityOverride()) {
    bOverrideWorldGravity=true
    overridePhysicsAsset=&/F_MED_DirectContact_FaceAcc_Physics_Dop
    simulationSpace=BaseBoneSpace
    alphaBoolBlend {blendOption=Linear}
    baseBoneRef {boneName=head}
    componentAppliedLinearAccClamp=<1000 650 400>
    componentLinearAccScale=<0.6 0.6 0.4>
    componentLinearVelScale=<0.8 0.4 0.2>
    overrideWorldGravity {z=-980}
    simSpaceSettings {worldAlpha=0.2}
   }
   n21 = "Transform (Modify) Bone - Bone: dyn_Pigtail_BACE"(ComponentPose=@n20, Alpha=1, Translation="0,0,0", Rotation="0,-3,0", Scale="1,1,1") {
    rotationMode=BMM_Additive
    rotationSpace=BCS_BoneSpace
    alphaBoolBlend {blendOption=Linear}
    boneToModify {boneName=dyn_Pigtail_BACE}
    rotation {yaw=-3}
   }
   n22 = Get Head Fwd Bwd Rot RBAN() -> (Head Fwd Bwd Rot RBAN:double)
   n23 = "Constraint - dyn_Pigtail_1 (PP)"!(ComponentPose=Local To Component(LocalPose=Use cached pose 'InputPose'()), Alpha=1, ConstraintWeights_0=0.85, ConstraintWeights_1:float=0.15) -> (Pose) {
    alphaBoolBlend {blendOption=Linear}
    boneToModify {boneName=dyn_Pigtail_1}
    constraintSetup [{targetBone {boneName=head}}, {targetBone {boneName=spine_05}}]
    constraintWeights [1,1]
   }
   n24 = "Constraint - dyn_Pigtail_2 (PP)"!(ComponentPose=@n23, Alpha=1, ConstraintWeights_0=0.05, ConstraintWeights_1:float=0.95) -> (Pose) {
    alphaBoolBlend {blendOption=Linear}
    boneToModify {boneName=dyn_Pigtail_2}
    constraintSetup [{targetBone {boneName=head}}, {targetBone {boneName=spine_05}}]
    constraintWeights [1,1]
   }
   n25 = "Constraint - dyn_Pigtail_3 (PP)"!(ComponentPose=@n24, Alpha=1, ConstraintWeights_0=0, ConstraintWeights_1:float=1) -> (Pose) {
    alphaBoolBlend {blendOption=Linear}
    boneToModify {boneName=dyn_Pigtail_3}
    constraintSetup [{targetBone {boneName=head}}, {targetBone {boneName=spine_05}}]
    constraintWeights [1,1]
   }
   n26 = "Constraint - dyn_Pigtail_4 (P)"(ComponentPose=@n25, Alpha=1, ConstraintWeights_0=1) {
    alphaBoolBlend {blendOption=Linear}
    boneToModify {boneName=dyn_Pigtail_4}
    constraintSetup [{targetBone {boneName=spine_05}}]
    constraintWeights [1]
   }
   n27 = "Constraint - dyn_Pigtail_5 (P)"(ComponentPose=@n26, Alpha=1, ConstraintWeights_0=1) {
    alphaBoolBlend {blendOption=Linear}
    boneToModify {boneName=dyn_Pigtail_5}
    constraintSetup [{targetBone {boneName=spine_05}}]
    constraintWeights [1]
   }
   n28 = "Constraint - dyn_Pigtail_6 (P)"(ComponentPose=@n27, Alpha=1, ConstraintWeights_0=1) {
    alphaBoolBlend {blendOption=Linear}
    boneToModify {boneName=dyn_Pigtail_6}
    constraintSetup [{targetBone {boneName=spine_05}}]
    constraintWeights [1]
   }
   n29 = "Transform (Modify) Bone - Bone: dyn_Pigtail_1"(ComponentPose=@n28, Alpha=@n22, Translation="-0.586,-2.886,0.168", Rotation="0.010,-19.863,0.020", Scale="1,1,1") {
    rotationMode=BMM_Additive
    rotationSpace=BCS_BoneSpace
    translationMode=BMM_Additive
    translationSpace=BCS_BoneSpace
    alphaBoolBlend {blendOption=Linear}
    alphaScaleBiasClamp {bMapRange=true, inRange {max=0, min=-25}, outRange {max=0, min=1}}
    boneToModify {boneName=dyn_Pigtail_1}
    rotation=<rot 0.01 -19.863 0.02>
    translation=<-0.586 -2.886 0.168>
   }
   n30 = "Transform (Modify) Bone - Bone: dyn_Pigtail_2"(ComponentPose=@n29, Alpha=@n22, Translation="-0.319,0.593,2.548", Rotation="0.073,28.065,-0.084", Scale="1,1,1") {
    rotationMode=BMM_Additive
    rotationSpace=BCS_BoneSpace
    translationMode=BMM_Additive
    translationSpace=BCS_BoneSpace
    alphaBoolBlend {blendOption=Linear}
    alphaScaleBiasClamp {bMapRange=true, inRange {max=0, min=-25}, outRange {max=0, min=1}}
    boneToModify {boneName=dyn_Pigtail_2}
    rotation=<rot 0.073 28.065 -0.084>
    translation=<-0.319 0.593 2.548>
   }
   Save cached pose 'Normal Dynamics'(Pose=Blend Poses by bool(BlendPose_0=Component To Local(ComponentPose=@n21), BlendPose_1=Component To Local(ComponentPose=@n30), bActiveValue=Get bRigidBodyEnabled()) {blendType=Linear}) {cacheName=Normal Dynamics}
   n31 = "Constraint - dyn_Pigtail_1 (PP)"(ComponentPose=Local To Component(LocalPose=Use cached pose 'InputPose'()), Alpha=1, ConstraintWeights_0=0.7, ConstraintWeights_1=0.3) {
    alphaBoolBlend {blendOption=Linear}
    boneToModify {boneName=dyn_Pigtail_1}
    constraintSetup [{targetBone {boneName=head}}, {targetBone {boneName=spine_05}}]
    constraintWeights [1,1]
   }
   n32 = "Constraint - dyn_Pigtail_2 (PP)"(ComponentPose=@n31, Alpha=1, ConstraintWeights_0=0, ConstraintWeights_1=1) {
    alphaBoolBlend {blendOption=Linear}
    boneToModify {boneName=dyn_Pigtail_2}
    constraintSetup [{targetBone {boneName=head}}, {targetBone {boneName=spine_05}}]
    constraintWeights [1,1]
   }
   n33 = "Constraint - dyn_Pigtail_3 (PP)"(ComponentPose=@n32, Alpha=1, ConstraintWeights_0=0, ConstraintWeights_1=1) {
    alphaBoolBlend {blendOption=Linear}
    boneToModify {boneName=dyn_Pigtail_3}
    constraintSetup [{targetBone {boneName=head}}, {targetBone {boneName=spine_05}}]
    constraintWeights [1,1]
   }
   n34 = "Constraint - dyn_Pigtail_4 (P)"(ComponentPose=@n33, Alpha=1, ConstraintWeights_0=1) {
    alphaBoolBlend {blendOption=Linear}
    boneToModify {boneName=dyn_Pigtail_4}
    constraintSetup [{targetBone {boneName=spine_05}}]
    constraintWeights [1]
   }
   n35 = "Constraint - dyn_Pigtail_5 (P)"(ComponentPose=@n34, Alpha=1, ConstraintWeights_0=1) {
    alphaBoolBlend {blendOption=Linear}
    boneToModify {boneName=dyn_Pigtail_5}
    constraintSetup [{targetBone {boneName=spine_05}}]
    constraintWeights [1]
   }
   n36 = "Constraint - dyn_Pigtail_6 (P)"(ComponentPose=@n35, Alpha=1, ConstraintWeights_0=1) {
    alphaBoolBlend {blendOption=Linear}
    boneToModify {boneName=dyn_Pigtail_6}
    constraintSetup [{targetBone {boneName=spine_05}}]
    constraintWeights [1]
   }
   n37 = Get head_fwdBwd_rot() -> (head_fwdBwd_rot:double)
   n38 = "Transform (Modify) Bone - Bone: dyn_Pigtail_1"(ComponentPose=@n36, Alpha=@n37, Translation="-0.586,-2.886,0.168", Rotation="0.010,-19.863,0.020", Scale="1,1,1") {
    rotationMode=BMM_Additive
    rotationSpace=BCS_BoneSpace
    translationMode=BMM_Additive
    translationSpace=BCS_BoneSpace
    alphaBoolBlend {blendOption=Linear}
    alphaScaleBiasClamp {bMapRange=true, inRange {max=0, min=-25}, outRange {max=0, min=1}}
    boneToModify {boneName=dyn_Pigtail_1}
    rotation=<rot 0.01 -19.863 0.02>
    translation=<-0.586 -2.886 0.168>
   }
   n39 = "Transform (Modify) Bone - Bone: dyn_Pigtail_2"(ComponentPose=@n38, Alpha=@n37, Translation="-0.930,2.043,-2.180", Rotation="0.068,22.906,-0.106", Scale="1,1,1") {
    rotationMode=BMM_Additive
    rotationSpace=BCS_BoneSpace
    translationMode=BMM_Additive
    translationSpace=BCS_BoneSpace
    alphaBoolBlend {blendOption=Linear}
    alphaScaleBiasClamp {bMapRange=true, inRange {max=0, min=-25}, outRange {max=0, min=1}}
    boneToModify {boneName=dyn_Pigtail_2}
    rotation=<rot 0.068 22.906 -0.106>
    translation=<-0.93 2.043 -2.18>
   }
   n40 = "Transform (Modify) Bone - Bone: dyn_tail_R_BACE"(ComponentPose=@n39, Alpha=Get ADS Up Side Braid Fix(), Translation="-1.673,-0.001,0.211", Rotation="-20.542,-0.020,0.003", Scale="1,1,1") {
    rotationMode=BMM_Additive
    rotationSpace=BCS_BoneSpace
    scaleMode=BMM_Additive
    scaleSpace=BCS_BoneSpace
    translationMode=BMM_Additive
    translationSpace=BCS_BoneSpace
    alphaBoolBlend {blendOption=Linear}
    boneToModify {boneName=dyn_tail_R_BACE}
    rotation=<rot -20.542 -0.0201 0.003>
    translation=<-1.673 -0.001 0.211>
   }
   n41 = "Transform (Modify) Bone - Bone: dyn_Pigtail_4"(ComponentPose=@n40, Alpha=1, Translation="0.003,2.134,1.203", Rotation="-0.013,2.085,0.025", Scale="1,1,1") {
    rotationMode=BMM_Additive
    rotationSpace=BCS_BoneSpace
    translationMode=BMM_Additive
    translationSpace=BCS_BoneSpace
    alphaBoolBlend {blendOption=Linear}
    alphaScaleBiasClamp {inRange {max=15}}
    boneToModify {boneName=dyn_Pigtail_4}
    rotation=<rot -0.013 2.085 0.025>
    translation=<0.003 2.134 1.203>
   }
   n42 = RigidBody(ComponentPose=@n41, OverrideWorldGravity=Get GravityOverride()) {
    bOverrideWorldGravity=true
    overridePhysicsAsset=&F_MED_DirectContact_FaceAcc_Physics_Dop
    simulationSpace=BaseBoneSpace
    alphaBoolBlend {blendOption=Linear}
    baseBoneRef {boneName=head}
    componentAppliedLinearAccClamp=<1000 650 400>
    componentLinearAccScale=<0.4 0.6 0.4>
    componentLinearVelScale=<0.8 0.4 0.2>
    overrideWorldGravity {z=-980}
    simSpaceSettings {worldAlpha=0.2}
   }
   n43 = Get isSkydiveOrParachuting() -> (isSkydiveOrParachuting:double)
   n44 = "Transform (Modify) Bone - Bone: dyn_Pigtail_3"!(ComponentPose=@n42, Alpha=@n43, Rotation="0,0,0", Scale="0.9,1,1") -> (Pose) {scaleMode=BMM_Additive, scaleSpace=BCS_BoneSpace, alphaBoolBlend {blendOption=Linear}, boneToModify {boneName=dyn_Pigtail_3}, scale {x=0.9}}
   n45 = "Transform (Modify) Bone - Bone: dyn_Pigtail_4"!(ComponentPose=@n44, Alpha=@n43, Rotation="0,-20,0", Scale="0.8,1,1") -> (Pose) {
    rotationMode=BMM_Additive
    rotationSpace=BCS_BoneSpace
    scaleMode=BMM_Additive
    scaleSpace=BCS_BoneSpace
    alphaBoolBlend {blendOption=Linear}
    boneToModify {boneName=dyn_Pigtail_4}
    rotation {yaw=-20}
    scale {x=0.8}
   }
   Save cached pose 'Backpack Dynamics'(Pose=Component To Local(ComponentPose=@n45)) {cacheName=Backpack Dynamics}
   n46 = "Constraint - dyn_Pigtail_1 (PP)"(ComponentPose=Local To Component(LocalPose=Use cached pose 'InputPose'()), Alpha=1, ConstraintWeights_0=0.85, ConstraintWeights_1=0.15) {
    alphaBoolBlend {blendOption=Linear}
    boneToModify {boneName=dyn_Pigtail_1}
    constraintSetup [{targetBone {boneName=head}}, {targetBone {boneName=spine_05}}]
    constraintWeights [1,1]
   }
   n47 = "Constraint - dyn_Pigtail_2 (PP)"(ComponentPose=@n46, Alpha=1, ConstraintWeights_0=0.05, ConstraintWeights_1=0.95) {
    alphaBoolBlend {blendOption=Linear}
    boneToModify {boneName=dyn_Pigtail_2}
    constraintSetup [{targetBone {boneName=head}}, {targetBone {boneName=spine_05}}]
    constraintWeights [1,1]
   }
   n48 = "Constraint - dyn_Pigtail_3 (PP)"(ComponentPose=@n47, Alpha=1, ConstraintWeights_0=0, ConstraintWeights_1=1) {
    alphaBoolBlend {blendOption=Linear}
    boneToModify {boneName=dyn_Pigtail_3}
    constraintSetup [{targetBone {boneName=head}}, {targetBone {boneName=spine_05}}]
    constraintWeights [1,1]
   }
   n49 = "Constraint - dyn_Pigtail_4 (P)"(ComponentPose=@n48, Alpha=1, ConstraintWeights_0=1) {
    alphaBoolBlend {blendOption=Linear}
    boneToModify {boneName=dyn_Pigtail_4}
    constraintSetup [{targetBone {boneName=spine_05}}]
    constraintWeights [1]
   }
   n50 = "Constraint - dyn_Pigtail_5 (P)"(ComponentPose=@n49, Alpha=1, ConstraintWeights_0=1) {
    alphaBoolBlend {blendOption=Linear}
    boneToModify {boneName=dyn_Pigtail_5}
    constraintSetup [{targetBone {boneName=spine_05}}]
    constraintWeights [1]
   }
   n51 = "Constraint - dyn_Pigtail_6 (P)"(ComponentPose=@n50, Alpha=1, ConstraintWeights_0=1) {
    alphaBoolBlend {blendOption=Linear}
    boneToModify {boneName=dyn_Pigtail_6}
    constraintSetup [{targetBone {boneName=spine_05}}]
    constraintWeights [1]
   }
   n52 = Get head_fwdBwd_rot()
   n53 = "Transform (Modify) Bone - Bone: dyn_Pigtail_1"(ComponentPose=@n51, Alpha=@n52, Translation="-0.586,-2.886,0.168", Rotation="0.010,-19.863,0.020", Scale="1,1,1") {
    rotationMode=BMM_Additive
    rotationSpace=BCS_BoneSpace
    translationMode=BMM_Additive
    translationSpace=BCS_BoneSpace
    alphaBoolBlend {blendOption=Linear}
    alphaScaleBiasClamp {bMapRange=true, inRange {max=0, min=-25}, outRange {max=0, min=1}}
    boneToModify {boneName=dyn_Pigtail_1}
    rotation=<rot 0.01 -19.863 0.02>
    translation=<-0.586 -2.886 0.168>
   }
   n54 = "Transform (Modify) Bone - Bone: dyn_Pigtail_2"(ComponentPose=@n53, Alpha=@n52, Translation="-0.930,2.043,-2.180", Rotation="0.068,22.906,-0.106", Scale="1,1,1") {
    rotationMode=BMM_Additive
    rotationSpace=BCS_BoneSpace
    translationMode=BMM_Additive
    translationSpace=BCS_BoneSpace
    alphaBoolBlend {blendOption=Linear}
    alphaScaleBiasClamp {bMapRange=true, inRange {max=0, min=-25}, outRange {max=0, min=1}}
    boneToModify {boneName=dyn_Pigtail_2}
    rotation=<rot 0.068 22.906 -0.106>
    translation=<-0.93 2.043 -2.18>
   }
   n55 = "Transform (Modify) Bone - Bone: dyn_Pigtail_4"!(ComponentPose=@n54, Alpha=1, Translation:Vector="0.003,2.134,1.203", Rotation="-0.013,2.085,0.025", Scale="1,1,1") -> (Pose) {
    rotationMode=BMM_Additive
    rotationSpace=BCS_BoneSpace
    translationMode=BMM_Additive
    translationSpace=BCS_BoneSpace
    alphaBoolBlend {blendOption=Linear}
    alphaScaleBiasClamp {inRange {max=15}}
    boneToModify {boneName=dyn_Pigtail_4}
    rotation=<rot -0.013 2.085 0.025>
    translation=<0.003 2.134 1.203>
   }
   n56 = "Transform (Modify) Bone - Bone: dyn_tail_R_BACE"(ComponentPose=@n55, Alpha=Get ADS Up Side Braid Fix(), Translation="-1.673,-0.001,0.211", Rotation="-20.542,-0.020,0.003", Scale="1,1,1") {
    rotationMode=BMM_Additive
    rotationSpace=BCS_BoneSpace
    scaleMode=BMM_Additive
    scaleSpace=BCS_BoneSpace
    translationMode=BMM_Additive
    translationSpace=BCS_BoneSpace
    alphaBoolBlend {blendOption=Linear}
    boneToModify {boneName=dyn_tail_R_BACE}
    rotation=<rot -20.542 -0.0201 0.003>
    translation=<-1.673 -0.001 0.211>
   }
   n57 = RigidBody(ComponentPose=@n56, OverrideWorldGravity=Get GravityOverride()) {
    bOverrideWorldGravity=true
    overridePhysicsAsset=&F_MED_DirectContact_FaceAcc_Physics_Dop
    simulationSpace=BaseBoneSpace
    alphaBoolBlend {blendOption=Linear}
    baseBoneRef {boneName=head}
    componentAppliedLinearAccClamp=<1000 650 400>
    componentLinearAccScale=<0.4 0.6 0.4>
    componentLinearVelScale=<0.8 0.4 0.2>
    overrideWorldGravity {z=-980}
    simSpaceSettings {worldAlpha=0.2}
   }
   Save cached pose 'Cape Dynamics'(Pose=Component To Local(ComponentPose=@n57)) {cacheName=Cape Dynamics}
  ]}
 EventGraph {nodes [
   Event OnCharacterCustomizationLoaded()
   > n1 = Cast To PlayerPawn_Athena(Object=TryGetPawnOwner())
   > Set Body_Mesh(Body_Mesh=GetSkeletalMeshForPartType(self=@n1))
   Event BlueprintInitializeAnimation()
   > Set Skel Mesh(Skel Mesh=GetOwnerSkeletalMeshComponent())
   Event BlueprintUpdateAnimation()
   > n2 = Sequence!(execute) -> (then_0, then_1, then_2:Exec, then_3:Exec, then_4:Exec, then_5:Exec, then_6:Exec, then_7:Exec)
   > n3 = Gravity Override State Function DEPRECATED(
    CharacterPartAnimInstance=Self-Reference()
    GameTimeInSeconds=GetGameTimeInSeconds()
    GravityOverrideParams=Get Gravity Override Params()
    DebugJointName=spine_05
   )
   > Set GravityOverride(GravityOverride=@n3)
   n4 = Get Skel Mesh()
   Set Hair Up(
    execute=@n2.then_1
    Hair Up:double=GetAlphaFromReferenceTransformDelta(SkeletalMeshComponent=@n4, BoneName=spine_05, BaseName=head, InputRange=(X=0,Y=70), TwistAxis="0,0,1")
   ) -> (Output_Get:double)
   > Set CrounchingFixHeadUp(CrounchingFixHeadUp:double=GetAlphaFromReferenceTransformDelta(SkeletalMeshComponent=@n4, BoneName=spine_05, BaseName=head, InputRange=(X=45,Y=62), TwistAxis="0,0,1")) -> (Output_Get:double)
   > Set Head Up Side Braid(Head Up Side Braid:double=GetAlphaFromReferenceTransformDelta(SkeletalMeshComponent=@n4, BoneName=spine_05, BaseName=head, InputRange=(X=30,Y=60), TwistAxis="0,0,1")) -> (Output_Get:double)
   Set ADS Up Side Braid Fix(
    execute=@n2.then_7
    ADS Up Side Braid Fix:double=float * float(A=Get Head Up Side Braid() -> (Head Up Side Braid:double), B:bool=Get bIsTargeting()) {is=title K2Node_PromotableOperator}
   ) -> (Output_Get:double)
   n5 = Get CrounchingFixHeadUp()
   n6 = Get bIsCrouching()
   > Set ADS Up Crounching(ADS Up Crounching:double=float * float(A=float * float(A=@n5, B=@n6), B=Get bIsTargeting())) -> (Output_Get:double)
   > Set ADS Up Crouch Moving(ADS Up Crouch Moving:double=float * float(A=Get ADS Up Crounching(), B=Get bIsMoving2D())) -> (Output_Get:double)
   n7 = Get bIsCrouching()
   n8 = Branch(execute=@n2.then_6, Condition="Nearly Equal (Float)"(A="To Float (Boolean)"(InBool=@n7), B=1, ErrorTolerance=0.1))
   > Set Head Fwd Bwd Rot Crouching(Head Fwd Bwd Rot Crouching:double=0) -> (Output_Get:double)
   n9 = GetDeltaTransformFromRefPose(self=GetOwnerSkeletalMeshComponent(), BoneName=head, BaseName=neck_01)
   Set Head Fwd Bwd Rot Crouching(execute=@n8.else, Head Fwd Bwd Rot Crouching=BreakRotator(InRot=BreakTransform(InTransform=@n9).Rotation).Yaw)
   n10 = Sequence!(execute=@n2.then_4) -> (then_0, then_1)
   n11 = GetDeltaTransformFromRefPose(self=GetOwnerSkeletalMeshComponent(), BoneName=head, BaseName=neck_01)
   Set Head Fwd Bwd Rot RBAN(execute=[@n10.then_0, @n13.then], Head Fwd Bwd Rot RBAN:double=BreakRotator(InRot=BreakTransform(InTransform=@n11).Rotation).Yaw) -> (Output_Get:double)
   n12 = Get bRigidBodyEnabled()
   n13 = Branch(execute=@n10.then_1, Condition="Nearly Equal (Float)"(A="To Float (Boolean)"(InBool=@n12), B=1, ErrorTolerance=0.1))
   Set Head Fwd Bwd Rot RBAN(execute=@n13.else, Head Fwd Bwd Rot RBAN=0)
   n14 = Get is Wearing Backbling()
   n15 = Branch(execute=@n2.then_3, Condition="Nearly Equal (Float)"(A="To Float (Boolean)"(InBool=@n14), B=1, ErrorTolerance=0.1))
   n16 = GetDeltaTransformFromRefPose(self=GetOwnerSkeletalMeshComponent(), BoneName=head, BaseName=neck_01)
   > Set head_fwdBwd_rot(head_fwdBwd_rot:double=BreakRotator(InRot=BreakTransform(InTransform=@n16).Rotation).Yaw) -> (Output_Get:double)
   Set head_fwdBwd_rot(execute=@n15.else, head_fwdBwd_rot=0)
   n17 = CheckForBackbling()
   Set is Wearing Backbling(
    execute=@n2.then_2
    is Wearing Backbling:bool=AND Boolean!(A=@n17, B=NOT Boolean(A=@n17.bIsBackblingACape), C:bool=NOT Boolean(A=@n17.bIsBackblingLong)) -> (ReturnValue)
   ) -> (Output_Get:bool)
   > Set is Wearing Cape(is Wearing Cape:bool=OR Boolean(A=@n17.bIsBackblingACape, B=@n17.bIsBackblingLong)) -> (Output_Get:bool)
   n18 = Get bIsSkydiveDiveMode()
   n19 = Get bIsSkydiveFloating()
   > n20 = Set is Skydiving(is Skydiving:double="To Float (Boolean)"(InBool=OR Boolean(A=@n18, B=@n19))) -> (Output_Get:double)
   > n21 = Set is Parachuting Alpha(is Parachuting Alpha:double="To Float (Boolean)"(InBool=Get bIsParachuteOpened())) -> (Output_Get:double)
   > Set isSkydiveOrParachuting(isSkydiveOrParachuting:double=float + float(A=@n21, B=@n20)) -> (Output_Get:double)
   n22 = PlayerMovement(execute=@n2.then_5, AnimInstance=Self-Reference(), XDotProductCoefficient=100, YDotProductCoefficient=100, ZDotProductCoefficient=100)
   n23 = Get bIsCrouching()
   n24 = "float < float"(A=BreakVector(InVec=MakeVector(X=@n22, Y=@n22.YMovement)), B=-50)
   > Set Left Crouching(Left Crouching:double=float * float(A="To Float (Boolean)"(InBool=@n23), B:double="To Float (Boolean)"(InBool=@n24))) -> (Output_Get:double)
   n25 = Get bIsKneeSliding()
   n26 = Get bIsBaseballSliding()
   n27 = Get bIsCrouching()
   > Set is Crouching(is Crouching:double="To Float (Boolean)"(InBool=OR Boolean!(A=@n27, B=@n26, C:bool=@n25) -> (ReturnValue))) -> (Output_Get:double)
  ]}
]
variables [
 ADS Up Crouch Moving {type=number}
 ADS Up Crounching {type=number}
 ADS Up Side Braid Fix {type=number}
 Body_Mesh {type=SkeletalMeshComponent}
 CrounchingFixHeadUp {type=number}
 Custom Base Mesh {type=SkeletalMeshComponent}
 Gravity Override Params {type=GravityOverrideParamsStruct}
 GravityOverride {type=Vector}
 Hair Up {type=number}
 Head Fwd Bwd Rot Crouching {type=number}
 Head Fwd Bwd Rot RBAN {type=number}
 Head Up Side Braid {type=number}
 head_fwdBwd_rot {type=number}
 is Crouching {type=number}
 is Parachuting Alpha {type=number}
 is Skydiving {type=number}
 is Wearing Backbling {type=boolean}
 Is wearing cape {type=boolean}
 isSkydiveOrParachuting {type=number}
 Left Crouching {type=number}
 RBAN Params {
  type=RBANParamsStruct
  default {
   dBNO {simSpaceSettings {
     maxAngularAcceleration=1000
     maxAngularVelocity=3
     maxLinearAcceleration=300
     maxLinearVelocity=700
     velocityScaleZ=0
     externalLinearDragV=<0 0.5 0>
    }}
   dBNO_Carried {simSpaceSettings {
     maxAngularAcceleration=1000
     maxAngularVelocity=3
     maxLinearAcceleration=300
     maxLinearVelocity=700
     velocityScaleZ=0
     externalLinearDragV=<0 0.5 0>
    }}
   emoteOrMelee {simSpaceSettings {
     maxAngularAcceleration=1000
     maxAngularVelocity=3
     maxLinearAcceleration=300
     maxLinearVelocity=700
     velocityScaleZ=0
     externalLinearDragV=<0 0.5 0>
    }}
   front_End {simSpaceSettings {
     maxAngularAcceleration=1000
     maxAngularVelocity=3
     maxLinearAcceleration=300
     maxLinearVelocity=700
     velocityScaleZ=0
     externalLinearDragV=<0 0.5 0>
    }}
   onGround_Crouch_Moving {simSpaceSettings {
     maxAngularAcceleration=1000
     maxAngularVelocity=3
     maxLinearAcceleration=300
     maxLinearVelocity=700
     velocityScaleZ=0
     externalLinearDragV=<0 0.5 0>
    }}
   onGround_Falling {simSpaceSettings {
     maxAngularAcceleration=1000
     maxAngularVelocity=3
     maxLinearAcceleration=300
     maxLinearVelocity=700
     velocityScaleZ=0
     externalLinearDragV=<0 0.5 0>
    }}
   onGround_Moving {simSpaceSettings {
     maxAngularAcceleration=1000
     maxAngularVelocity=3
     maxLinearAcceleration=300
     maxLinearVelocity=700
     velocityScaleZ=0
     externalLinearDragV=<0 0.5 0>
    }}
   onGround_Standing {simSpaceSettings {
     maxAngularAcceleration=1000
     maxAngularVelocity=3
     maxLinearAcceleration=300
     maxLinearVelocity=700
     velocityScaleZ=0
     externalLinearDragV=<0 0.5 0>
    }}
   rail_Grinding {simSpaceSettings {
     maxAngularAcceleration=1000
     maxAngularVelocity=3
     maxLinearAcceleration=300
     maxLinearVelocity=700
     velocityScaleZ=0
     worldAlpha=0.1
     externalLinearDragV=<0 0.5 0>
    }}
   riding_Moving {simSpaceSettings {
     maxAngularAcceleration=1000
     maxAngularVelocity=3
     maxLinearAcceleration=300
     maxLinearVelocity=700
     velocityScaleZ=0
     externalLinearDragV=<0 0.5 0>
    }}
   riding_Moving_Targeting {simSpaceSettings {
     maxAngularAcceleration=1000
     maxAngularVelocity=3
     maxLinearAcceleration=300
     maxLinearVelocity=700
     velocityScaleZ=0
     externalLinearDragV=<0 0.5 0>
    }}
   riding_Standing {simSpaceSettings {
     maxAngularAcceleration=1000
     maxAngularVelocity=3
     maxLinearAcceleration=300
     maxLinearVelocity=700
     velocityScaleZ=0
     externalLinearDragV=<0 0.5 0>
    }}
   skydive_DivingDown {simSpaceSettings {
     maxAngularAcceleration=1000
     maxAngularVelocity=3
     maxLinearAcceleration=300
     maxLinearVelocity=700
     velocityScaleZ=0
     externalLinearDragV=<0 0.5 0>
    }}
   skydive_FloatingDown {
    componentLinearAccScale=<0 0.5 0>
    componentLinearVelScale=<0 0.5 0>
    simSpaceSettings {
     maxAngularAcceleration=1000
     maxAngularVelocity=3
     maxLinearAcceleration=300
     maxLinearVelocity=700
     velocityScaleZ=0
     externalLinearDragV=<0 0.5 0>
    }
   }
   skydive_GoingUp {simSpaceSettings {
     maxAngularAcceleration=1000
     maxAngularVelocity=3
     maxLinearAcceleration=300
     maxLinearVelocity=700
     velocityScaleZ=0
     externalLinearDragV=<0 0.5 0>
    }}
   skydive_Parachuting {
    componentLinearAccScale=<0 0.1 0.3>
    componentLinearVelScale=<0 0.1 0.3>
    simSpaceSettings {
     maxAngularAcceleration=1000
     maxAngularVelocity=3
     maxLinearAcceleration=300
     maxLinearVelocity=700
     velocityScaleZ=0
     externalLinearDragV=<0 0.5 0>
    }
   }
   skydive_Surfing {simSpaceSettings {
     maxAngularAcceleration=1000
     maxAngularVelocity=3
     maxLinearAcceleration=300
     maxLinearVelocity=700
     velocityScaleZ=0
     externalLinearDragV=<0 0.5 0>
    }}
   swimming {simSpaceSettings {
     maxAngularAcceleration=1000
     maxAngularVelocity=3
     maxLinearAcceleration=300
     maxLinearVelocity=700
     velocityScaleZ=0
     externalLinearDragV=<0 0.5 0>
    }}
   tactical_Sprint {simSpaceSettings {
     maxAngularAcceleration=1000
     maxAngularVelocity=3
     maxLinearAcceleration=300
     maxLinearVelocity=700
     velocityScaleZ=0
     externalLinearDragV=<0 0.5 0>
    }}
   ziplining {simSpaceSettings {
     maxAngularAcceleration=1000
     maxAngularVelocity=3
     maxLinearAcceleration=300
     maxLinearVelocity=700
     velocityScaleZ=0
     externalLinearDragV=<0 0.5 0>
    }}
  }
 }
 RBAN Settings {
  type=RBANSettingStruct
  default {
   componentLinearAccScale=<0 0 0>
   componentLinearVelScale=<0 0 0>
   simSpaceSettings {
    dampingAlpha=1
    maxAngularAcceleration=1000
    maxAngularVelocity=3
    maxLinearAcceleration=300
    maxLinearVelocity=700
    velocityScaleZ=0
    worldAlpha=1
    externalAngularVelocity=<0 0 0>
    externalLinearDragV=<0 0.5 0>
    externalLinearVelocity=<0 0 0>
   }
  }
 }
 Skel Mesh {type=SkeletalMeshComponent}
]
